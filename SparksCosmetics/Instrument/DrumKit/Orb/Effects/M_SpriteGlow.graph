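# Translucent, two-sided unlit Niagara sprite and mesh particle material for a glow in the drum kit orb effects.
# Particle color drives adjusted emissive; particle alpha scales a mask made by subtracting a fixed inner radial gradient from an adjustable one.
# Switches select the plain mask or one or two panning radial-UV noise textures, with optional background, sampled using original-UV derivatives.
# Dynamic parameters drive radius, tiling, panning and power, though only X tiling and pan values are used.
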Material /SparksCosmetics/Instrument/DrumKit/Orb/Effects/M_SpriteGlow
graph {
 properties {
  bAutomaticallySetUsageInEditor=false
  bUsedWithNiagaraMeshParticles=true
  bUsedWithNiagaraSprites=true
  blendMode=BLEND_Translucent
  pixelDepthOffsetMode=PDOM_Legacy
  shadingModel=MSM_Unlit
  twoSided=true
  usedShadingModels=Unlit
 }
 nodes [
  n1 = ParticleColor()
  MaterialFunctionCall(Emissive=@n1.RGB) {materialFunction=&MF_FX_EmissiveAdjustments}
  n2 = DynamicParameter() {parameterIndex=1, defaultValue=<rgba 0.4 1.5 -0.1 -0.1>, paramNames [Radius,Density,PannerX,PannerY]}
  n3 = Constant() {r=0.3}
  n4 = Constant() {r=4}
  n5 = Subtract(
   A=MaterialFunctionCall(Radius=@n2.Radius, Density=@n2.Radius) {materialFunction=&RadialGradientExponential}.RadialGradientExponential
   B=Saturate(None=MaterialFunctionCall(Radius=@n3, Density=@n4) {materialFunction=&RadialGradientExponential}.RadialGradientExponential)
  )
  n6 = Saturate(None=@n5)
  n7 = MaterialFunctionCall() {materialFunction=&VectorToRadialValue}
  n8 = DynamicParameter() {defaultValue=<rgba 3 1 1 1>, paramNames [TillingX,TillingY,TillingX02,TillingY02]}
  n9 = AppendVector(A=@n8.TillingX, B=@n8.TillingX)
  n10 = DynamicParameter() {parameterIndex=1, defaultValue=<rgba 0.4 1.5 -0.1 -0.1>, paramNames [Radius,Density,PannerX,PannerY]}
  n11 = AppendVector(A=@n10.PannerX, B=@n10.PannerX)
  n12 = TextureCoordinate()
  n13 = DDX(Value=@n12)
  n14 = DDY(Value=@n12)
  n15 = TextureSampleParameter2D!(
   UVs=Panner(Coordinate=Multiply(A=@n7.Radial Coordinates, B=@n9), Speed=@n11) {speedX=-0.1, speedY=-0.1}
   "DDX(UVs)"=@n13
   "DDY(UVs)"=@n14
   Apply View MipBias
  ) {
   mipValueMode=TMVM_Derivative
   parameterName=Texture01
   samplerType=SAMPLERTYPE_Grayscale
   texture=/FortUILibrary/UIKit/Buttons/Textures/LowResBlurredNoise
  }
  n16 = AppendVector(A=@n8.TillingX02, B=@n8.TillingX02)
  n17 = TextureSampleParameter2D(UVs=Panner(Coordinate=Multiply(A=@n7.Radial Coordinates, B=@n16), Speed=@n11) {speedX=-0.1, speedY=-0.1}, "DDX(UVs)"=@n13, "DDY(UVs)"=@n14) {
   mipValueMode=TMVM_Derivative
   parameterName=Texture02
   samplerType=SAMPLERTYPE_Grayscale
   texture=&LowResBlurredNoise
  }
  n18 = Power(
   Base=StaticSwitchParameter(True=@n15.R, False=Add(A=@n15.R, B=@n17.R)) {parameterName=OneTextureOr2Textures}
   Exp=DynamicParameter() {parameterIndex=2, defaultValue=<rgba 3 1 1 1>, paramNames [Power,Param2,Param3,Param4]}.Power
  ) {constExponent=3}
  n19 = Multiply(A=@n18, B=@n5)
  n20 = Multiply(A=Add(A=@n18, B=@n6), B=@n5)
  Multiply(
   A=@n1.A
   B=StaticSwitchParameter(True=@n6, False=StaticSwitchParameter(True=@n19, False=@n20) {parameterName=TextureOrTextureWithBackground}) {parameterName=JustMaskOrMaskAndNoise}
  )
 ]
}
